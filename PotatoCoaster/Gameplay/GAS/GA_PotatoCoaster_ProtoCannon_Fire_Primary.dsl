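; GA_PotatoCoaster_ProtoCannon_Fire_Primary is a gameplay ability for the ProtoCannon primary fire in PotatoCoaster.
; The activation event has no logic attached in this graph.
; When the ability ends, whether cancelled or not, it sends a Protocannon AmmoChanged gameplay event by tag.
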(event Ability|EventActivateAbility)

(event Ability|EventOnEndAbility
  (params (bWasCancelled :type Boolean))
  (Ability|SendGameplayEvent "(TagName=\"Event.Athena.Item.Protocannon.AmmoChanged\")"))
